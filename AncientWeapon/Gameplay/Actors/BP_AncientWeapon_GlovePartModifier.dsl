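; Character part modifier for the Ancient Weapon glove, driving the gauntlet's visuals alongside the thrown weapon.
; On post-res-in it casts the pawn's current weapon to the thrown weapon, registers itself there, triggers equip emissive and sets custom primitive data for gem and line glow.
; Recall runs emissive plus a local burst GameplayCue; catch directly activates and socket-attaches a Niagara component as a log-spam workaround.
; Weapon changes re-attach idle VFX next tick.
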(event Custom|GloveRecallGC
  ;[0.4383 0.1161 1] RECALL GC
  (Class|BChronoWeapAncientWeaponThrown|GauntletEmissiveonRecall (Variables|Default|GetAs%20B%20Chrono%20Weap%20Ancient%20Thrown))
  (GameplayCue|ExecuteGameplayCue%28Burst%29Local self (Variables|Default|GetRecall%20Gameplay%20Cue%20Tag) (Ability|GameplayCue|MakeGameplayCueParameters :NormalizedMagnitude 0.0 :RawMagnitude 0.0 :Location (Transformation|GetSocketLocation (CustomCharacter|GetSkeletalMeshForPartOnPlayerPawn) (Variables|Default|GetRecall%20GC%20FX%20Socket%20Name)) :Normal "0,0,0" :GameplayEffectLevel 1 :AbilityLevel 1 :TargetAttachComponent (CustomCharacter|GetSkeletalMeshForPartOnPlayerPawn) :bReplicateLocationWhenUsingMinimalRepProxy false)))

(event Custom|GloveCatchGC
  (bind _glove_equip_ns (Variables|Default|GetGlove%20Equip%20NS))
  ;[0.4383 0.1161 1] CATCH GC
  ;;[1 1 1] This is a work around for FORT-589563. A systemic fix described in FORT-617814 should make the original code work without log spam. Note also this work around required the addition of the Glove Equip NS component.
  (Components|Activation|Activate _glove_equip_ns false)
  (Transformation|AttachComponentToComponent _glove_equip_ns (CustomCharacter|GetSkeletalMeshForPartOnPlayerPawn) (Variables|Default|GetCatch%20GC%20FX%20Socket%20Name) "KeepRelative" "KeepRelative" "KeepRelative" false))

(event EventOnPostResIn
  (|CallParentFunction)
  (bind _asfort_player_pawn_athena (Utilities|Casting|CastToFortPlayerPawnAthena (CustomCharacter|GetAssociatedPlayerPawn)))
  (bind _asb_chrono_weap_ancient_weapon_thrown (Utilities|Casting|CastToB_Chrono_Weap_AncientWeapon_Thrown (Variables|/Script/FortniteGame.FortPawn|GetCurrentWeapon _asfort_player_pawn_athena)))
  (bind _output_get (Variables|Default|SetAs%20B%20Chrono%20Weap%20Ancient%20Thrown _asb_chrono_weap_ancient_weapon_thrown))
  (Variables|/AncientWeapon/Gameplay/Actors/B_Chrono_Weap_AncientWeapon_Thrown.B_Chrono_Weap_AncientWeapon_Thrown_C|SetGlove%20Part%20Modifier self _output_get)
  (Class|BChronoWeapAncientWeaponThrown|GauntletEmissiveonEquip _output_get)
  (Rendering|Material|SetCustomPrimitiveDataFloat (CustomCharacter|GetSkeletalMeshForPartOnPlayerPawn) 1 2.0) ; 0-1 ramps up gem brightness and 1-2 fills lines
)

(event EventOnDayPhaseChanged
  (params (CurrentDayPhase :type EFortDayPhase%20Enum :default "Morning") (PreviousDayPhase :type EFortDayPhase%20Enum :default "Morning") (bAtCreation :type Boolean)))

(event Custom|CustomEvent
  (params (Pawn :type Fort%20Player%20Pawn%20Object%20Reference) (Part :type Custom%20Character%20Part%20Object%20Reference) (MeshPart :type Skeletal%20Mesh%20Component%20Object%20Reference) (PartType :type EFortCustomPartType%20Enum :default "Head")))

(event EventOnNewPlayerWeaponType
  (params (NewType :type EFortItemType%20Enum :default "WorldItem") (OldType :type EFortItemType%20Enum :default "WorldItem"))
  (|CallParentFunction NewType OldType) ; This node is disabled and will not be called.\nDrag off pins to build functionality.
)

(event EventOnWeaponChanged
  (params (NewWeapon :type Fort%20Weapon%20Object%20Reference) (OldWeapon :type Fort%20Weapon%20Object%20Reference))
  (Utilities|FlowControl|DelayUntilNextTick)
  (Transformation|AttachComponentToComponent (CallFunction|GetIdleVFXComponent) (CustomCharacter|GetSkeletalMeshForPartOnPlayerPawn) "FX_Socket" "SnapToTarget" "SnapToTarget" "KeepRelative" true))

(event EventOnWeaponFired)

(event EventOnNewWeapons
  (params (WeaponsAsItems :type Array%20of%20Fort%20Item%20Object%20References)))

(event EventOnCurieStateAttached
  (params (Tag :type Gameplay%20Tag%20Structure%20%28by%20ref%29)))

(event EventOnCurieStateDetached
  (params (Tag :type Gameplay%20Tag%20Structure%20%28by%20ref%29)))

(event EventOnBeginSkydiving
  (|CallParentFunction) ; This node is disabled and will not be called.\nDrag off pins to build functionality.
)

(event EventOnEndSkydiving
  (|CallParentFunction) ; This node is disabled and will not be called.\nDrag off pins to build functionality.
)

(event EventOnEnterVehicle
  (params (VehicleEntered :type Fort%20Vehicle%20Interface%20Interface%20%28by%20ref%29))
  (|CallParentFunction VehicleEntered) ; This node is disabled and will not be called.\nDrag off pins to build functionality.
)

(event EventOnExitVehicle
  (|CallParentFunction) ; This node is disabled and will not be called.\nDrag off pins to build functionality.
)

(event EventOnTargetingOrFirstPersonCameraChanged)

(event EventOnHitPawn
  (params (SelfActor :type Actor%20Object%20Reference) (OtherActor :type Actor%20Object%20Reference) (NormalImpulse :type Vector :default "0, 0, 0") (Hit :type Hit%20Result%20Structure%20%28by%20ref%29)))

(fn ConstructionScript
  (|CallParentFunction))
