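; Weapon actor for the Ancient Weapon thrown boomerang and gauntlet, handling its equip, throw, recall, catch and cooldown presentation.
; Self-restarting async gameplay-tag watchers hide the boomerang during sprint and clamber and reshow it on slide, through an authority-set RepNotify bool.
; Timelines, custom primitive data and Niagara parameters drive emissive charge effects.
; Throw and projectile state control a timer-updated reticle with hit flash, while visibility and vehicle exit feed idle audio.
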(event Custom|BoomerangEmissiveOnEquip
  (bind _ns_ancientweapon_ambient (Variables|Default|GetNS_AncientWeapon_Ambient))
  ;[0.9281 0.3654 1] Comment
  (Utilities|FlowControl|Sequence
    (:then_0
      (Utilities|FlowControl|Delay 0.25)
      (Niagara|SetNiagaraVariable(Bool) _ns_ancientweapon_ambient "Charge" true)
      (Utilities|FlowControl|Delay 0.75) ; Allows time for a timeline dependent\neffect to play in NS_AncientWeapon_Ambient\nbefore resetting
      (Niagara|SetNiagaraVariable(Bool) _ns_ancientweapon_ambient "Charge" false))
    (:then_1
      (bind _weaponmesh (Variables|Visual|GetWeaponMesh)
        (:then
          (CallFunction|PlayBoomerangEquipTimeline _weaponmesh (Variables|Default|GetNS_AncientWeapon_Ambient)))
        (:else)))))

(event Custom|GauntletEmissiveOnEquip
  ;[0.9281 0.3654 1] Comment
  (bind _glove_part_modifier (Variables|Default|GetGlove%20Part%20Modifier)
    (:then
      (Class|BPAncientWeaponGlovePartModifier|GloveCatchGC _glove_part_modifier)
      (CallFunction|PlayGauntletEquipTimeline _glove_part_modifier))
    (:else)))

(event Custom|BoomernagEmissiveOnCooldownEnded
  (params (Skip%20Timeline :type Boolean))
  ;[0.9281 0.3654 1] Comment
  (bind _weaponmesh (Variables|Visual|GetWeaponMesh)
    (:then
      (if Skip%20Timeline
        (Rendering|Material|SetCustomPrimitiveDataFloat _weaponmesh 0 3.0)
        (Niagara|SetNiagaraVariable(Float) (Variables|Default|GetNS_AncientWeapon_Ambient) "ChargeAmount" 3.0)))
    (:else)))

(event Custom|BoomerangEmissiveOnCooldown
  (bind _ns_ancientweapon_ambient (Variables|Default|GetNS_AncientWeapon_Ambient))
  ;[0.9281 0.3654 1] Boomerang Emissive zero after catch
  (bind _weaponmesh (Variables|Visual|GetWeaponMesh)
    (:then
      (Rendering|Material|SetCustomPrimitiveDataFloat _weaponmesh 0 0.0)
      (Niagara|SetNiagaraVariable(Float) _ns_ancientweapon_ambient "ChargeAmount" 0.0)
      (Niagara|SetNiagaraVariable(Bool) _ns_ancientweapon_ambient "Charge" true)
      (Utilities|FlowControl|Delay (Math|Float|Clamp(Float) (- (ScalableFloat|GetValueatLevel (Variables|Default|GetMinCooldownForThrow)) 0.2) 0.0 (ScalableFloat|GetValueatLevel (Variables|Default|GetMinCooldownForThrow))))
      (CallFunction|BoomernagEmissiveonCooldownEnded false)
      (Utilities|FlowControl|Delay 0.75) ; Allows time for an effect to play in NS_AncientWeapon_Ambient 
      (Niagara|SetNiagaraVariable(Bool) (Variables|Default|GetNS_AncientWeapon_Ambient) "Charge" false))
    (:else)))

(event Custom|GauntletEmissiveOnThrow
  ;[0.9281 0.3654 1] Gauntlet emissive fill after throw
  (bind _glove_part_modifier (Variables|Default|GetGlove%20Part%20Modifier)
    (:then
      (CallFunction|StopGauntletCatchTimeline)
      (CallFunction|StopGauntletEquipTimeline)
      (Rendering|Material|SetCustomPrimitiveDataFloat (CustomCharacter|GetSkeletalMeshForPartOnPlayerPawn _glove_part_modifier) 0 (Variables|Default|GetGauntlet%20Emissive%20Off%20Value)))
    (:else)))

(event Custom|GauntleEmissiveOnCatch
  (params (Skip%20Timeline :type Boolean))
  ;[0.9281 0.3654 1] Gauntlet emissive on Catch
  (bind _glove_part_modifier (Variables|Default|GetGlove%20Part%20Modifier)
    (:then
      (CallFunction|StopGauntletRecallTimeline)
      (Class|BPAncientWeaponGlovePartModifier|GloveCatchGC _glove_part_modifier)
      (if Skip%20Timeline
        (bind _glove_part_modifier_2 (Variables|Default|GetGlove%20Part%20Modifier)
          (:then
            (Rendering|Material|SetCustomPrimitiveDataFloat (CustomCharacter|GetSkeletalMeshForPartOnPlayerPawn _glove_part_modifier_2) 0 (Variables|FortWeaponRangedChronoAncient|GetGauntletEmissiveMaxValue)))
          (:else))
        (else
          (CallFunction|PlayGauntletCatchTimeline _glove_part_modifier))))
    (:else)))

(event Custom|WatchForClamberTagToHideWeaponMesh
  (bind _playerpawn (Variables|Default|GetPlayerPawn))
  ;[1 1 1] Keep glove out. Hide boomerang on clamber/mantle
  (bind _asynctaskproxy (Ability|Tasks|WaitGameplayTagAddToActor _playerpawn (Variables|Default|GetClamber%20State%20Tag) true)
    (:then
      (Variables|Default|SetClamber%20Start%20Async%20Action _asynctaskproxy))
    (:Added
      (CallFunction|SetMeshVisibilityRepNotifyonAuthority false)
      (bind _asynctaskproxy_2 (Ability|Tasks|WaitGameplayTagRemoveFromActor _playerpawn (Variables|Default|GetClamber%20State%20Tag) true)
        (:then
          (Variables|Default|SetClamber%20Stop%20Async%20Action _asynctaskproxy_2))
        (:Removed
          (CallFunction|SetMeshVisibilityRepNotifyonAuthority true)
          (CallFunction|WatchforClamberTagtoHideWeaponMesh))))))

(event Custom|GauntletEmissiveOnRecall
  ;[0.9281 0.3654 1] Gauntlet emissive on Recall
  (bind _glove_part_modifier (Variables|Default|GetGlove%20Part%20Modifier)
    (:then
      (CallFunction|PlayGauntletEquipTimeline _glove_part_modifier))
    (:else)))

(event Custom|WatchForSprintTagToHideWeaponMesh
  (bind _in_flight_tag (Variables|Default|GetIn%20Flight%20Tag))
  (bind _playerpawn (Variables|Default|GetPlayerPawn))
  (bind _returnvalue (GameplayTags|HasMatchingGameplayTag _playerpawn _in_flight_tag))
  (bind _ancientweaponsprinttag (Variables|Default|GetAncientWeaponSprintTag))
  (bind _playerpawn_2 (Variables|Default|GetPlayerPawn))
  (bind _returnvalue_2 (GameplayTags|HasMatchingGameplayTag _playerpawn_2 _ancientweaponsprinttag))
  (bind _returnvalue_3 (not _returnvalue_2))
  (bind _in_flight_tag_2 (Variables|Default|GetIn%20Flight%20Tag))
  (bind _playerpawn_3 (Variables|Default|GetPlayerPawn))
  (bind _returnvalue_4 (GameplayTags|HasMatchingGameplayTag _playerpawn_3 _in_flight_tag_2))
  (bind _returnvalue_5 (not _returnvalue_4))
  (bind _returnvalue_6 (and _returnvalue_3 _returnvalue_5))
  (bind _playerpawn_4 (Variables|Default|GetPlayerPawn))
  ;[1 1 1] Keep glove out. Hide on Sprint, Show on Slide / end Sprint.
  (bind _asynctaskproxy (Ability|Tasks|WaitGameplayTagAddToActor _playerpawn_4 (Variables|Default|GetAncientWeaponSprintTag) true)
    (:then
      (Variables|Default|SetAsync%20Sprint%20Hide%20Weapon%20Action _asynctaskproxy)
      (bind _asynctaskproxy_2 (Ability|Tasks|WaitGameplayTagAddToActor _playerpawn_4 (Variables|Default|GetAncientWeaponSlideTag) true)
        (:then
          (Variables|Default|SetAsync%20Slide%20Start%20Show%20Weapon%20%20Action _asynctaskproxy_2))
        (:Added
          (bind _async_sprint_show_weapon_action (Variables|Default|GetAsync%20Sprint%20Show%20Weapon%20Action)
            (:then
              (Ability|Async|EndAction _async_sprint_show_weapon_action)
              ;;[1 1 1] If in flight, don't show mesh
              (CallFunction|SetMeshVisibilityRepNotifyonAuthority (not _returnvalue))
              (if (not _returnvalue)
                (CallFunction|HandleIdleSoundLoop true))
              (bind _asynctaskproxy_3 (Ability|Tasks|WaitGameplayTagRemoveFromActor _playerpawn_4 (Variables|Default|GetAncientWeaponSlideTag) true)
                (:then
                  ;;[1 1 1] If not in flight & not sprinting when slide ends, show visibility
                  (Variables|Default|SetAsync%20Slide%20End%20Action _asynctaskproxy_3))
                (:Removed
                  (CallFunction|SetMeshVisibilityRepNotifyonAuthority _returnvalue_6)
                  ;;[1 1 1] End existing asyncs if they are still valid before restarting singleton watcher
                  (if _returnvalue_6
                    (CallFunction|HandleIdleSoundLoop true)
                    (else
                      _))
                  (bind _async_sprint_show_weapon_action_2 (Variables|Default|GetAsync%20Sprint%20Show%20Weapon%20Action)
                    (:then
                      (Ability|Async|EndAction _async_sprint_show_weapon_action_2))
                    (:else
                      _)))))
            (:else)))))
    (:Added
      (CallFunction|SetMeshVisibilityRepNotifyonAuthority false)
      (CallFunction|HandleIdleSoundLoop false)
      (bind _asynctaskproxy_4 (Ability|Tasks|WaitGameplayTagRemoveFromActor _playerpawn_4 (Variables|Default|GetAncientWeaponSprintTag) true)
        (:then
          (Variables|Default|SetAsync%20Sprint%20Show%20Weapon%20Action _asynctaskproxy_4))
        (:Removed
          (if (not (GameplayTags|HasMatchingGameplayTag (Variables|Default|GetPlayerPawn) (Variables|Default|GetIn%20Flight%20Tag)))
            (CallFunction|SetMeshVisibilityRepNotifyonAuthority true))
          (CallFunction|HandleIdleSoundLoop true)
          (bind _async_slide_start_show_weapon__action (Variables|Default|GetAsync%20Slide%20Start%20Show%20Weapon%20%20Action)
            (:then
              (Ability|Async|EndAction _async_slide_start_show_weapon__action))
            (:else
              _))
          (bind _async_slide_end_action (Variables|Default|GetAsync%20Slide%20End%20Action)
            (:then
              (Ability|Async|EndAction _async_slide_end_action))
            (:else
              _))))))
  (CallFunction|WatchforSprintTagtoHideWeaponMesh))

(event FortWeapon|EventOnEquippedAndAttached
  ;[1 1 1] Unhide the actor if it's hidden. TODO: selective based on if we caught boomerang in away or equipped state... also try to avoid equip anim on away state
  (Variables|Default|SetIsAncientWeapon true)
  (bind _asfort_player_pawn (Utilities|Casting|CastToFortPlayerPawn (Game|GetInstigator)))
  (CallFunction|VehicleAudioSet)
  (bind _output_get (Variables|Default|SetPlayerPawn _asfort_player_pawn))
  (bind _returnvalue (GameplayTags|HasAnyMatchingGameplayTags(Message) _output_get (Variables|Default|GetAncientWeaponInFlightTag)))
  (Utilities|FlowControl|Sequence
    (:then_0
      (if (Pawn|IsLocallyControlled _output_get)
        (Networking|CanExecuteCosmeticEvents
          (:True
            ;;[1 1 1] If in flight, not visible & if in hand, visible
            (CallFunction|ToggleBoomerangThrowPrompt (not _returnvalue))))))
    (:then_1
      (if (not _returnvalue)
        ;;[1 1 1] FORT-614441 - remove tag if player drops item, catches, then switches back, otherwise throw is blocked until GE expitres
        (GameplayEffects|RemoveActiveEffectswithGrantedTags (Ability|GetAbilitySystemComponent (Variables|Default|GetPlayerPawn)) (Variables|Default|GetBlock%20Throw%20Tag))
        ;;[1 1 1] Need to Set Visibility for Catch without boomerang equipped & switch back case
        (Utilities|FlowControl|Sequence
          (:then_0
            (Utilities|FlowControl|Sequence
              (:then_0
                _)
              (:then_1
                (Components|Activation|Activate (Variables|Default|GetNS_AncientWeapon_Ambient) false)
                (Utilities|FlowControl|Delay 0.2) ; Visual emphasis on equip
                (CallFunction|GauntletEmissiveonEquip)
                (Utilities|FlowControl|Delay 0.15) ; Visual emphasis on equip
                (CallFunction|BoomerangEmissiveonEquip))))
          (:then_1
            (CallFunction|SetMeshVisibilityRepNotifyonAuthority (CallFunction|ShouldShowWeaponMeshifNotinFlightonEquip)))))
      (CallFunction|WatchforSprintTagtoHideWeaponMesh)
      (CallFunction|WatchforClamberTagtoHideWeaponMesh))))

(event FortWeapon|EventOnLocalUnEquip
  ;[1 1 1] Unhide the actor if it's hidden. TODO: selective based on if we caught boomerang in away or equipped state... also try to avoid equip anim on away state
  (Variables|Default|SetIsAncientWeapon false)
  (Components|Activation|Deactivate (Variables|Default|GetNS_AncientWeapon_Ambient))
  (Ability|Async|EndAction (Variables|Default|GetAsync%20Sprint%20Show%20Weapon%20Action))
  (Default|UpdateReticleVisibility)
  (Utilities|Time|ClearandInvalidateTimerbyHandle (Variables|Default|GetReticleTickTimer)))

(event Custom|UnhideWeaponMeshAll
  (Development|SetHiddeninGame (Variables|Visual|GetWeaponMesh) false true))

(event Custom|ReticleTick
  ;[1 1 1] Reticle State Updates
  (Default|ReticleUpdate))

(event Custom|HandleProjectileHit
  ;[0.1669 0 1] HandleProjectileHit(): calls Client_HandleProjectileHit()
  ; Client_HandleProjectileHit(): plays a timeline that shows the ReticleHitColor going from 1 to 0.
  (Networking|SwitchHasAuthority
    (:Authority
      (CallFunction|ClientHandleProjectileHit))))

(event Custom|Client%20HandleProjectileHit
  ;[0.1669 0 1] HandleProjectileHit(): calls Client_HandleProjectileHit()
  ; Client_HandleProjectileHit(): plays a timeline that shows the ReticleHitColor going from 1 to 0.
  (Networking|CanExecuteCosmeticEvents
    (:True
      (bind (_direction _hitpercent) (Timeline|Timeline PlayFromStart)
        (:Update
          (Variables|Default|SetReticleHitPercent _hitpercent))
        (:Finished
          (Variables|Default|SetReticleHitPercent 0.0))))))

(event Custom|ToggleBoomerangThrowPrompt
  (params (Visible :type Boolean))
  ;[0.4828 0 1] Toggle Throw Prompt
  (Variables|Default|SetIsInThrowableState Visible)
  (Default|UpdateReticleVisibility))

(event Custom|Set%20Projetile%20Event
  (params (B%20Prj%20Ancient%20Weapon :type B%20Prj%20Ancient%20Weapon%20Boomerang%20Home%20Object%20Reference))
  ;[0.04076 0.1108 1] SetProjectile
  (Variables|Default|SetProjectileReplicated B%20Prj%20Ancient%20Weapon)
  (Variables|Default|SetHomingTarget 0) ; Clear homing target after proj is spawned
)

(event EventOnWeaponVisibilityChanged
  (params (bVisible :type Boolean) (bSetForLocalControllerOnly :type Boolean))
  ;[0 0 0] Idle audio logic
  (|CallParentFunction bVisible bSetForLocalControllerOnly)
  (CallFunction|HandleIdleSoundLoop (and bVisible (> (Math|Vector|VectorLength (Transformation|GetActorScale3D)) 1.0))))

(event Custom|HandleIdleSoundLoop
  (params (bShouldPlay :type Boolean))
  (bind _idleaudio (Variables|Default|GetIdleAudio))
  ;[0 0 0] Idle audio logic
  (Variables|Default|GetIdleAudio
    (:then
      (if bShouldPlay
        (if (Audio|Components|Audio|IsPlaying (Variables|Default|GetIdleAudio))
          (Audio|Components|Audio|FadeOut _idleaudio 0.3 0.0 "Linear" :!class /Script/Engine.AudioComponent))
        (Audio|Components|Audio|FadeIn _idleaudio 0.3 1.0 0.0 "Linear" :!class /Script/Engine.AudioComponent)
        (else
          (Audio|Components|Audio|FadeOut (Variables|Default|GetIdleAudio) 0.3 0.0 "Linear" :!class /Script/Engine.AudioComponent))))
    (:else)))

(event Custom|VehicleAudioSet
  ;[0 0 0] Idle audio logic
  (bind _asplayer_pawn_athena_generic_parent (Utilities|Casting|CastToPlayerPawn_Athena_Generic_Parent (Game|GetInstigator)))
  (bind _output_get (Variables|Audio|SetWeaponOwnerPawn _asplayer_pawn_athena_generic_parent))
  (Default|BindEventtoOnExitedVehicleDispatcher _output_get (delegate Custom|OnExitedVehicle)))

(event Custom|OnExitedVehicle
  ;[0 0 0] Idle audio logic
  (if (not (Audio|Components|Audio|IsPlaying (Variables|Default|GetIdleAudio)))
    (CallFunction|HandleIdleSoundLoop true)))

(event Custom|VehicleAudioUnset
  ;[0 0 0] Idle audio logic
  (bind _weaponownerpawn (Variables|Audio|GetWeaponOwnerPawn)
    (:then
      (Default|UnbindEventfromOnExitedVehicleDispatcher _weaponownerpawn (delegate Custom|OnExitedVehicle)))
    (:else)))

(event FortWeapon|EventOnUnEquippedAndDetached
  ;[0 0 0] Idle audio logic
  (CallFunction|VehicleAudioUnset))

(fn AmmoChargedGC
  (bind _asfort_pawn (Utilities|Casting|CastToFortPawn (Variables|Actor|GetInstigator)))
  (if (Pawn|IsLocallyControlled _asfort_pawn)
    (GameplayCue|ExecuteGameplayCue%28Burst%29Local _asfort_pawn (Variables|Default|GetAmmoChargeCue) (Ability|GameplayCue|MakeGameplayCueParameters 0.0))))

(fn OnRep_bOnRepWeaponMeshVisible
  (bind _weaponmesh (Variables|Visual|GetWeaponMesh))
  (if (Variables|Default|GetbOnRepWeaponMeshVisible)
    (Rendering|SetVisibility _weaponmesh true true)
    (else
      (Rendering|SetVisibility _weaponmesh false true))))

(fn OnRep_ProjectileReplicated
  (Networking|CanExecuteCosmeticEvents
    (:True
      (Default|UpdateReticleVisibility))))

(fn ReticleUpdate
  ;[1 1 1] Check if we have objects we can hit in homing range. If player, update that bool for UI specific case
  (Variables|Default|GetProjectileReplicated
    (:then
      _)
    (:else
      (bind _asfort_player_pawn (Utilities|Casting|CastToFortPlayerPawn (Game|GetInstigator)))
      (if (Utilities|IsValid _asfort_player_pawn :!class /Script/Engine.KismetSystemLibrary)
        (LocalVariables|SetEnvTargetInRange (<= (Math|Vector|Distance(Vector) (Transformation|GetActorLocation (Game|GetInstigator)) (Reticle|GetLocationUnderReticle (Fortnite|GetFortPlayerControllerfromActor _asfort_player_pawn))) (ScalableFloat|GetValueatLevel (Variables|Default|GetRow_HomingTargetingMaxRange))))
        (LocalVariables|SetPawnTargetInRange (.bSuccess (Utilities|Casting|CastToFortPlayerPawn (Reticle|GetActorUnderReticle (Fortnite|GetFortPlayerControllerfromActor _asfort_player_pawn))))))))
  (Default|CallUpdateReticleState (LocalVariables|GetPawnTargetInRange) (LocalVariables|GetEnvTargetInRange)))

(fn SetAmbientAudioState
  (params (bEnabled :type Boolean))
  (if bEnabled
    (if (and (not (GameplayTags|HasAnyMatchingGameplayTags (Utilities|Casting|CastToFortPlayerPawn (Variables|Actor|GetInstigator)) (Variables|Default|GetTC_ActionsThatPreventIdleAudio))) (Variables|Default|GetbEquipped))
      (Variables|Default|GetAmbientAudioComp
        (:then
          _)
        (:else
          (bind _returnvalue (Audio|SpawnSoundAttached (Variables|Audio|GetSB_MaceAmbient_Loop) (Variables|Visual|GetWeaponMesh)))
          (bind _output_get (Variables|Default|SetAmbientAudioComp _returnvalue))))
      (Audio|Parameter|SetIntegerParameter _output_get "PowerLevel" (FortWeapon|GetMagazineAmmoCount) :!class /Script/AudioExtensions.AudioParameterControllerInterface))
    (else
      (Audio|Components|Audio|FadeOut (Variables|Default|GetAmbientAudioComp) 0.1 0.0 "Linear" :!class /Script/Engine.AudioComponent))))

(fn SetMeshVisibilityRepNotifyOnAuthority
  (params (SetVisible :type Boolean))
  (Utilities|FlowControl|Sequence
    (:then_0
      (Networking|SwitchHasAuthority
        (:Authority
          (Variables|Default|SetbOnRepWeaponMeshVisible SetVisible))))
    (:then_1
      (return))))

(fn ShouldShowWeaponMeshIfNotInFlightOnEquip
  (return (or (not (GameplayTags|HasMatchingGameplayTag (Variables|Default|GetPlayerPawn) (Variables|Default|GetAncientWeaponSprintTag))) (and (GameplayTags|HasMatchingGameplayTag (Variables|Default|GetPlayerPawn) (Variables|Default|GetAncientWeaponSprintTag)) (GameplayTags|HasMatchingGameplayTag (Variables|Default|GetPlayerPawn) (Variables|Default|GetAncientWeaponSlideTag))))))

(fn UpdateReticleVisibility
  (bind _projectilereplicated (Variables|Default|GetProjectileReplicated))
  (bind _returnvalue (Utilities|IsValid _projectilereplicated :!class /Script/Engine.KismetSystemLibrary))
  (bind _isinthrowablestate (Variables|Default|GetIsInThrowableState))
  (bind _returnvalue_2 (CallFunction|IsBeingEquipped))
  (bind _returnvalue_3 (FortWeapon|IsEquipped))
  (bind _returnvalue_4 (Math|Boolean|NORBoolean _returnvalue_3 _returnvalue_2))
  (bind _returnvalue_5 (Math|Boolean|ORBoolean _returnvalue_4 _isinthrowablestate _returnvalue))
  (bind _returnvalue_6 (not _returnvalue_5))
  (bind _hidereticletags (Variables|Default|GetHideReticleTags))
  (Networking|CanExecuteCosmeticEvents
    (:True
      (bind _playerpawn (Variables|Default|GetPlayerPawn)
        (:then
          (if (Pawn|IsLocallyControlled _playerpawn)
            (HUD|SetHUDElementVisibility (BlueprintContext|GetFortHUDContext) (Variables|Default|GetReticle%20HUD%20Element%20Tag) (or (GameplayTags|HasAnyMatchingGameplayTags _playerpawn _hidereticletags) _returnvalue_6) "None") ; Show the reticle when in the throw state and when the projectile is currently flying
            (if (or (GameplayTags|HasAnyMatchingGameplayTags _playerpawn _hidereticletags) _returnvalue_6)
              (Utilities|Time|ClearandInvalidateTimerbyHandle (Variables|Default|GetReticleTickTimer))
              (else
                (bind _returnvalue_7 (Utilities|Time|SetTimerbyEvent (EventDispatchers|CreateEvent :!event ReticleTick) 0.01 true))
                (Variables|Default|SetReticleTickTimer _returnvalue_7)))))
        (:else)))))

(fn ConstructionScript
  (|CallParentFunction))
